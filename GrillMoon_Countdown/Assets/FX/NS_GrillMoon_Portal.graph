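# Niagara system for the GrillMoon countdown portal, whose look changes through seven phases indexed by System.Phase, switching from pink to gold at phase 6.
# Per-phase float arrays gate spawning, sizes and material parameters, swapping background meshes from translucent to opaque to masked.
# Center sprites carry a looping sound, joined by flickering light, gold-phase mesh ribbons and GPU sparks swirled by curl noise and vortex forces.
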NiagaraSystem /GrillMoon_Countdown/Assets/FX/NS_GrillMoon_Portal
dependencies {
 usedDataInterfaces [
  &NiagaraDataInterfaceArrayFloat
  &NiagaraDataInterfaceCurve
  &NiagaraDataInterfaceCurlNoise
  &NiagaraDataInterfaceColorCurve
 ]
 usedDynamicInputs [&SimulationPosition, &Multiply_Vector2D_ByFloat]
 usedModules [
  &self:SystemSpawnScript.SystemScriptSource.SystemScriptGraph.NiagaraNodeAssignment_0.SetVariables_#1
  &SystemState
  &self:SystemSpawnScript.SystemScriptSource.SystemScriptGraph.NiagaraNodeAssignment_1.SetVariables_#2
 ]
}
emitters {
 BG_Masked {
  emitterUpdateScript [
   &EmitterState(Life Cycle Mode=::NewEnumerator1, Loop Duration=1, UseLoopDelay=false)
   &SpawnBurst_Instantaneous(
    Spawn Probability=&SelectFloatFromArray(Direct Array Index=$System.Phase, Float Selection Array=DataInterface() {propertyValues {FloatData [0,0,0,0,0,0,1]}})
    Use Spawn Probability=true
    UseLoopCountLimit=false
   )
   &SpawnRate(SpawnRate=20, Use Spawn Probability=false) {enabled=false}
  ]
  particleSpawnScript [
   &InitializeParticle(Mesh Scale Mode=::NewEnumerator3, Mesh Uniform Scale=5, Position Offset=<0 0 50>, UsePositionOffset=true, Write Material Random=true) {module=InitializeParticle001}
   /Niagara/Modules/Spawn/Initialization/InitializeParticle(
    Mesh Scale=&VectorFromFloat(Value=4.5)
    Module.Write Color=true
    Module.Write Lifetime=true
    Module.Write Mass=true
    Module.Write Material Random=true
    Module.Write Position=true
    Module.Write Scale=true
    Module.Write SpriteRotation=false
    Module.Write SpriteSize=true
    Position=&AddVectorToPosition(Position=/Niagara/DynamicInputs/Helpers/SimulationPosition(), Vector=<0 0 50>)
    Sprite Size=<5 5>
   ) {enabled=false}
   &InitialMeshOrientation(
    Mesh Axis To Orient=<0 0 1>
    Mesh Orientation Mode=::NewEnumerator2
    Orientation Coordinate Space=::Local
    Rotation=<-0.0025 0 0>
    Use Rotation Vector=true
   )
  ]
  particleUpdateScript [
   &ParticleState()
   &SolveForcesAndVelocity(Clamp Velocity=false, Limit Acceleration=false)
   &DynamicMaterialParameters(
    Index 0 Param 1=0.5
    Index 0 Param 2=&MakeCustomFloatFromBool(Bool=&SetBoolByIntComparison(A=$System.Phase, B=6, Comparison Type=::NewEnumerator4), True Float=1, False Float=0)
    Module.Index0Param0Write=true
    Module.Index0Param1Write=true
    Module.Index0Param2Write=true
    Module.Index0Param3Write=true
   )
  ]
  rendererClasses [&NiagaraSpriteRendererProperties, &NiagaraMeshRendererProperties]
  renderers [
   {rendererClass=&NiagaraSpriteRendererProperties}
   {rendererClass=&NiagaraMeshRendererProperties, rendererIndex=1}
  ]
 }
 BG_Opaque {
  emitterUpdateScript [
   &EmitterState(Life Cycle Mode=::NewEnumerator1, Loop Duration=1, UseLoopDelay=false)
   &SpawnBurst_Instantaneous(
    Spawn Probability=&SelectFloatFromArray(Direct Array Index=$System.Phase, Float Selection Array=DataInterface() {propertyValues {FloatData [0,0,0,0,1,1,0]}})
    Use Spawn Probability=true
    UseLoopCountLimit=false
   )
   &SpawnRate(SpawnRate=20, Use Spawn Probability=false) {enabled=false}
  ]
  particleSpawnScript [
   /Niagara/Modules/Spawn/Initialization/V2/InitializeParticle(Mesh Scale Mode=::NewEnumerator3, Mesh Uniform Scale=4.5, Position Offset=<0 0 50>, UsePositionOffset=true, Write Material Random=true) {module=InitializeParticle001}
   /Niagara/Modules/Spawn/Initialization/InitializeParticle(
    Mesh Scale=&VectorFromFloat(Value=4.5)
    Module.Write Color=true
    Module.Write Lifetime=true
    Module.Write Mass=true
    Module.Write Material Random=true
    Module.Write Position=true
    Module.Write Scale=true
    Module.Write SpriteRotation=false
    Module.Write SpriteSize=true
    Position=&AddVectorToPosition(Position=&SimulationPosition(), Vector=<0 0 50>)
    Sprite Size=<5 5>
   ) {enabled=false}
   &InitialMeshOrientation(
    Mesh Axis To Orient=<0 0 1>
    Mesh Orientation Mode=::NewEnumerator2
    Orientation Coordinate Space=::Local
    Rotation=<-0.0025 0 0>
    Use Rotation Vector=true
   )
  ]
  particleUpdateScript [
   &ParticleState()
   &SolveForcesAndVelocity(Clamp Velocity=false, Limit Acceleration=false)
   &DynamicMaterialParameters(
    Index 0 Param 1=&SelectFloatFromArray(Direct Array Index=$System.Phase, Float Selection Array=DataInterface() {propertyValues {FloatData [100,1,0.5,0.3,0,0,1]}})
    Index 0 Param 2=&MakeCustomFloatFromBool(Bool=&SetBoolByIntComparison(A=$System.Phase, B=6, Comparison Type=::NewEnumerator4), True Float=1, False Float=0)
    Module.Index0Param0Write=true
    Module.Index0Param1Write=true
    Module.Index0Param2Write=true
    Module.Index0Param3Write=true
   )
  ]
  rendererClasses [&NiagaraSpriteRendererProperties, &NiagaraMeshRendererProperties]
  renderers [
   {rendererClass=&NiagaraSpriteRendererProperties}
   {rendererClass=&NiagaraMeshRendererProperties, rendererIndex=1}
  ]
 }
 BG_Translucent {
  emitterUpdateScript [
   &EmitterState(Life Cycle Mode=::NewEnumerator1, Loop Duration=1, UseLoopDelay=false)
   &SpawnBurst_Instantaneous(
    Spawn Probability=&SelectFloatFromArray(Direct Array Index=$System.Phase, Float Selection Array=DataInterface() {propertyValues {FloatData [1,1,1,1,0,0,0]}})
    Use Spawn Probability=true
    UseLoopCountLimit=false
   )
   &SpawnRate(SpawnRate=20, Use Spawn Probability=false) {enabled=false}
  ]
  particleSpawnScript [
   &InitializeParticle(
    Mesh Scale=&VectorFromFloat(Value=4.5)
    Module.Write Color=true
    Module.Write Lifetime=true
    Module.Write Mass=true
    Module.Write Material Random=true
    Module.Write Position=true
    Module.Write Scale=true
    Module.Write SpriteRotation=false
    Module.Write SpriteSize=true
    Position=&AddVectorToPosition(Position=&SimulationPosition(), Vector=<0 0 50>)
    Sprite Size=<5 5>
   ) {enabled=false}
   /Niagara/Modules/Spawn/Initialization/V2/InitializeParticle(Mesh Scale Mode=::NewEnumerator3, Mesh Uniform Scale=4.5, Position Offset=<0 0 50>, UsePositionOffset=true, Write Material Random=true) {module=InitializeParticle001}
   &InitialMeshOrientation(
    Mesh Axis To Orient=<0 0 1>
    Mesh Orientation Mode=::NewEnumerator2
    Orientation Coordinate Space=::Local
    Rotation=<-0.0025 0 0>
    Use Rotation Vector=true
   )
  ]
  particleUpdateScript [
   &ParticleState()
   &SolveForcesAndVelocity(Clamp Velocity=false, Limit Acceleration=false)
   &DynamicMaterialParameters(
    Index 0 Param 1=&SelectFloatFromArray(Direct Array Index=$System.Phase, Float Selection Array=DataInterface() {propertyValues {FloatData [100,1,0.5,0.3,0,0,0]}})
    Index 0 Param 2=&MakeCustomFloatFromBool(Bool=&SetBoolByIntComparison(A=$System.Phase, B=6, Comparison Type=::NewEnumerator4), True Float=1, False Float=0)
    Module.Index0Param0Write=true
    Module.Index0Param1Write=true
    Module.Index0Param2Write=true
    Module.Index0Param3Write=true
   )
  ]
  rendererClasses [&NiagaraSpriteRendererProperties, &NiagaraMeshRendererProperties]
  renderers [
   {rendererClass=&NiagaraSpriteRendererProperties}
   {rendererClass=&NiagaraMeshRendererProperties, rendererIndex=1}
  ]
 }
 CenterGold {
  emitterSpawnScript [
   &InitializeAudioPlayer(Audio Player:NiagaraDataInterfaceAudioPlayer=$User.Portal Looping Sound)
   /Niagara/Modules/Audio/PlayPersistentAudio()
  ]
  emitterUpdateScript [
   &EmitterState(Life Cycle Mode=::NewEnumerator1, Loop Duration=1, UseLoopDelay=false)
   &SpawnRate(
    SpawnRate=&SelectFloatFromArray(Direct Array Index=$System.Phase, Float Selection Array=DataInterface() {propertyValues {FloatData [0,0,0,0,0,0,0.75]}})
    Use Spawn Probability=false
   )
   /Niagara/Modules/Audio/UpdatePersistentAudio()
  ]
  particleSpawnScript [
   &InitializeParticle(
    Color=&MakeCustomLinearColorFromBool(
     Bool=&SetBoolByIntComparison(A=$System.Phase, B=6, Comparison Type=::NewEnumerator4)
     True Color=<rgba 2 1.4211 0.5602 1>
     False Color=<rgba 2 0.5521 0.9382 1>
    )
    Lifetime=8
    Mass Mode=::NewEnumerator1
    Position Offset=<0 0 50>
    Sprite Rotation Angle=&UniformRangedFloat(Minimum=0, Maximum=360)
    Sprite Rotation Mode=::NewEnumerator0
    Sprite Size Mode=::NewEnumerator3
    Uniform Sprite Size=&SelectFloatFromArray(Direct Array Index=$System.Phase, Float Selection Array=DataInterface() {propertyValues {FloatData [650,650,650,650,650,650,750]}})
    UsePositionOffset=false
    Write Material Random=true
   ) {module=InitializeParticle001}
   /SetVariables(
    Particles.Position=&Add_Vector(A=Linked() {linkedVariable {name=Particles.Position, type=NiagaraPosition}}, B=<0 0 0>)
    Particles.Random=&UniformRangedFloat(Minimum=-0.1, Maximum=0.9)
   ) {isSetParameters=true, module=SetVariables_#3}
  ]
  particleUpdateScript [
   &ParticleState()
   &ScaleSpriteSize(Scale Sprite Size Mode=::NewEnumerator0, Uniform Scale Factor=4)
   &SolveForcesAndVelocity(Clamp Velocity=false, Limit Acceleration=false)
   &DynamicMaterialParameters(
    Index 0 Param 1=&FloatFromCurve(FloatCurve=DataInterface() {propertyValues {Curve {keys [
         {arriveTangent=1.1486, interpMode=RCIM_Cubic, leaveTangent=1.1486, tangentMode=RCTM_User, time=0, value=0}
         {arriveTangent=-9.5651, interpMode=RCIM_Cubic, leaveTangent=-9.5651, tangentMode=RCTM_User, time=0.6446, value=-2.4156}
         {arriveTangent=-1.1502, interpMode=RCIM_Cubic, leaveTangent=-1.1502, tangentMode=RCTM_User, time=1, value=-3.0414}
        ]}}})
    Index 0 Param 2=&FloatFromCurve(FloatCurve=DataInterface() {propertyValues {Curve {keys [
         {arriveTangent=-10.815, interpMode=RCIM_Cubic, leaveTangent=-10.815, tangentMode=RCTM_User, time=0.0013, value=0.9969}
         {arriveTangent=-0.881, interpMode=RCIM_Cubic, leaveTangent=-0.881, time=0.3186, value=0.0305}
         {arriveTangent=2.4595, interpMode=RCIM_Cubic, leaveTangent=2.4595, time=0.758, value=0.3302}
         {arriveTangent=3.5097, interpMode=RCIM_Cubic, leaveTangent=3.5097, tangentMode=RCTM_User, time=1, value=1.7064}
        ]}}})
    Index 0 Param 4=&Add_Float(
     A=$Particles.Random
     B=&FloatFromCurve(
      FloatCurve=DataInterface() {propertyValues {Curve {keys [{interpMode=RCIM_Cubic, tangentMode=RCTM_User, time=0, value=0}, {interpMode=RCIM_Cubic, tangentMode=RCTM_User, time=1, value=1}]}}}
      Scale Curve=-1
     )
    )
    Index 1 Param 1=&SelectFloatFromArray(Direct Array Index=$System.Phase, Float Selection Array=DataInterface() {propertyValues {FloatData [100,20,10,5,0,0,1]}})
    Module.Param0WriteEnabled=true
    Module.Param0WriteEnabled1=true
    Module.Param1WriteEnabled=true
    Module.Param1WriteEnabled1=true
    Module.Param2WriteEnabled=true
    Module.Param2WriteEnabled1=true
    Module.Param3WriteEnabled=true
    Module.Param3WriteEnabled1=true
    Write Parameter Index 1=true
   )
   &ScaleColor(
    Scale Mode=::NewEnumerator1
    Scale RGBA=&Vector4FromCurve(Vector4Curve=DataInterface() {propertyValues {
       AlphaCurve {keys [
         <at 0.0037 0 RCIM_Cubic>
         {arriveTangent=2.9787, interpMode=RCIM_Cubic, leaveTangent=2.9787, time=0.1552, value=0.2595}
         {arriveTangent=1.0157, interpMode=RCIM_Cubic, leaveTangent=1.0157, time=0.2714, value=0.7974}
         {arriveTangent=-1.0205, interpMode=RCIM_Cubic, leaveTangent=-1.0205, time=0.7097, value=0.8228}
         {arriveTangent=-2.8344, interpMode=RCIM_Cubic, leaveTangent=-2.8344, time=0.8433, value=0.2138}
         <at 1 0 RCIM_Cubic>
        ]}
       BlueCurve {keys [<at 0.2703 1>,<at 0.8712 0>]}
       GreenCurve {keys [<at 0.2703 1>,<at 0.8712 0>]}
       RedCurve {keys [<at 0.2703 1>,<at 0.8712 0>]}
      }})
    ScaleRGBA=true
   )
   &self:MatchScale() {module=MatchScale}
  ]
  rendererClasses [&NiagaraSpriteRendererProperties]
  renderers [{rendererClass=&NiagaraSpriteRendererProperties}]
 }
 CenterPink {
  emitterSpawnScript [
   &InitializeAudioPlayer(Audio Player:NiagaraDataInterfaceAudioPlayer=$User.Portal Looping Sound)
   &PlayPersistentAudio()
  ]
  emitterUpdateScript [
   &EmitterState(Life Cycle Mode=::NewEnumerator1, Loop Duration=1, UseLoopDelay=false)
   &SpawnRate(
    SpawnRate=&SelectFloatFromArray(Direct Array Index=$System.Phase, Float Selection Array=DataInterface() {propertyValues {FloatData [0.75,0.75,0.75,0.75,0.75,0,0]}})
    Use Spawn Probability=false
   )
   &UpdatePersistentAudio()
  ]
  particleSpawnScript [
   &InitializeParticle(
    Color=&MakeCustomLinearColorFromBool(
     Bool=&SetBoolByIntComparison(A=$System.Phase, B=6, Comparison Type=::NewEnumerator4)
     True Color=<rgba 2 1.4211 0.5602 1>
     False Color=<rgba 2 0.5521 0.9382 1>
    )
    Lifetime=8
    Mass Mode=::NewEnumerator1
    Position Offset=<0 0 50>
    Sprite Rotation Angle=&UniformRangedFloat(Minimum=0, Maximum=360)
    Sprite Rotation Mode=::NewEnumerator0
    Sprite Size Mode=::NewEnumerator3
    Uniform Sprite Size=&SelectFloatFromArray(Direct Array Index=$System.Phase, Float Selection Array=DataInterface() {propertyValues {FloatData [650,650,650,650,650,650,750]}})
    UsePositionOffset=false
    Write Material Random=true
   ) {module=InitializeParticle001}
   /SetVariables(
    Particles.Position=&Add_Vector(A=Linked() {linkedVariable {name=Particles.Position, type=NiagaraPosition}}, B=<0 0 0>)
    Particles.Random=&UniformRangedFloat(Minimum=-0.1, Maximum=0.9)
   ) {isSetParameters=true, module=SetVariables_#3}
  ]
  particleUpdateScript [
   &ParticleState()
   &ScaleSpriteSize(Scale Sprite Size Mode=::NewEnumerator0, Uniform Scale Factor=4)
   &SolveForcesAndVelocity(Clamp Velocity=false, Limit Acceleration=false)
   &DynamicMaterialParameters(
    Index 0 Param 1=&FloatFromCurve(FloatCurve=DataInterface() {propertyValues {Curve {keys [
         {arriveTangent=1.1486, interpMode=RCIM_Cubic, leaveTangent=1.1486, tangentMode=RCTM_User, time=0, value=0}
         {arriveTangent=-9.5651, interpMode=RCIM_Cubic, leaveTangent=-9.5651, tangentMode=RCTM_User, time=0.6446, value=-2.4156}
         {arriveTangent=-1.1502, interpMode=RCIM_Cubic, leaveTangent=-1.1502, tangentMode=RCTM_User, time=1, value=-3.0414}
        ]}}})
    Index 0 Param 2=&FloatFromCurve(FloatCurve=DataInterface() {propertyValues {Curve {keys [
         {arriveTangent=-10.815, interpMode=RCIM_Cubic, leaveTangent=-10.815, tangentMode=RCTM_User, time=0.0013, value=0.9969}
         {arriveTangent=-0.881, interpMode=RCIM_Cubic, leaveTangent=-0.881, time=0.3186, value=0.0305}
         {arriveTangent=2.4595, interpMode=RCIM_Cubic, leaveTangent=2.4595, time=0.758, value=0.3302}
         {arriveTangent=3.5097, interpMode=RCIM_Cubic, leaveTangent=3.5097, tangentMode=RCTM_User, time=1, value=1.7064}
        ]}}})
    Index 0 Param 4=&Add_Float(
     A=$Particles.Random
     B=&FloatFromCurve(
      FloatCurve=DataInterface() {propertyValues {Curve {keys [{interpMode=RCIM_Cubic, tangentMode=RCTM_User, time=0, value=0}, {interpMode=RCIM_Cubic, tangentMode=RCTM_User, time=1, value=1}]}}}
      Scale Curve=2
     )
    )
    Index 1 Param 1=&SelectFloatFromArray(Direct Array Index=$System.Phase, Float Selection Array=DataInterface() {propertyValues {FloatData [100,20,10,5,0,0,1]}})
    Module.Param0WriteEnabled=true
    Module.Param0WriteEnabled1=true
    Module.Param1WriteEnabled=true
    Module.Param1WriteEnabled1=true
    Module.Param2WriteEnabled=true
    Module.Param2WriteEnabled1=true
    Module.Param3WriteEnabled=true
    Module.Param3WriteEnabled1=true
    Write Parameter Index 1=true
   )
   &ScaleColor(
    Scale Mode=::NewEnumerator1
    Scale RGBA=&Vector4FromCurve(Vector4Curve=DataInterface() {propertyValues {
       AlphaCurve {keys [
         <at 0.0037 0 RCIM_Cubic>
         {arriveTangent=2.9787, interpMode=RCIM_Cubic, leaveTangent=2.9787, time=0.1552, value=0.2595}
         {arriveTangent=1.0157, interpMode=RCIM_Cubic, leaveTangent=1.0157, time=0.2714, value=0.7974}
         {arriveTangent=-1.0205, interpMode=RCIM_Cubic, leaveTangent=-1.0205, time=0.7097, value=0.8228}
         {arriveTangent=-2.8344, interpMode=RCIM_Cubic, leaveTangent=-2.8344, time=0.8433, value=0.2138}
         <at 1 0 RCIM_Cubic>
        ]}
       BlueCurve {keys [<at 0.2703 1>,<at 0.8712 0>]}
       GreenCurve {keys [<at 0.2703 1>,<at 0.8712 0>]}
       RedCurve {keys [<at 0.2703 1>,<at 0.8712 0>]}
      }})
    ScaleRGBA=true
   )
   &self:MatchScale() {module=MatchScale}
  ]
  rendererClasses [&NiagaraSpriteRendererProperties]
  renderers [{rendererClass=&NiagaraSpriteRendererProperties}]
 }
 LightAndGlow {
  emitterUpdateScript [
   &EmitterState(
    Inactive Response=::NewEnumerator1
    Life Cycle Mode=::NewEnumerator1
    Loop Behavior=::NewEnumerator1
    Loop Duration Mode=::NewEnumerator1
    UseLoopDelay=false
   )
   &SpawnBurst_Instantaneous(
    Spawn Count=&MakeIntFromFloat(Float=&SelectFloatFromArray(Direct Array Index=$System.Phase, Float Selection Array=DataInterface() {propertyValues {FloatData [0,0,0,0,1,0,1]}}))
    Use Spawn Probability=false
    UseLoopCountLimit=false
   )
  ]
  particleSpawnScript [
   &InitializeParticle(
    Color=&MakeCustomLinearColorFromBool(
     Bool=&SetBoolByIntComparison(A=$System.Phase, B=6, Comparison Type=::NewEnumerator4)
     True Color=<rgba 1 0.2481 0 0.1>
     False Color=<rgba 0.6406 0.0606 0.2832 0.1>
    )
    Lifetime=5
    Position Offset=<100 0 200>
    Sprite Size Mode=::NewEnumerator3
    Uniform Sprite Size=3000
    UsePositionOffset=true
    Write Material Random=true
   )
   &FN_SetSpawnGroup(Number of Spawn Groups=3)
   &CameraOffset()
   &SpriteFacingAndAlignment(WriteAlignment=true, WriteFacing=true)
  ]
  particleUpdateScript [
   &ParticleState(Kill Particles When Lifetime Has Elapsed=false)
   &ScaleColor(
    Scale Alpha=&Lerp_Float(
     A=1
     B=&Add_Float(A=1, B=$System.CurrentLoopDuration)
     Alpha=&FloatFromCurve(
      FloatCurve=DataInterface() {propertyValues {Curve {keys [
          {interpMode=RCIM_Cubic, tangentMode=RCTM_User, time=0, value=0}
          {interpMode=RCIM_Cubic, tangentMode=RCTM_User, time=0.5, value=1}
          {tangentMode=RCTM_User, time=1, value=0}
         ]}}}
      CurveIndex=$System.NormalizedLoopAge
      Scale Curve=0.33
     )
    )
    ScaleA=true
    ScaleRGB=false
   )
   &FN_Light_Attributes(
    Color=&MakeCustomLinearColorFromBool(
     Bool=&SetBoolByIntComparison(A=$System.Phase, B=6, Comparison Type=::NewEnumerator4)
     True Color=<rgba 18 4.4652 0 1>
     False Color=<rgba 16 1.5124 7.0721 1>
    )
    Create Color Attribute=true
    Create Exponent Attribute=true
    Create Radius Attribute=true
    Exponent=2
    Radius=500
    Scale Color=&Waveform("Waveform [1]"=::NewEnumerator12, "[1] Frequency"=5, "[1] Amplitude Min/Max"=<0.8 1>)
   )
   &ScaleColor(
    Scale Alpha=&Waveform("Waveform [1]"=::NewEnumerator12, "[1] Frequency"=5, "[1] Amplitude Min/Max"=<0.8 1>)
    ScaleA=true
    ScaleRGB=true
   ) {module=ScaleColor001}
  ]
  rendererClasses [&NiagaraLightRendererProperties, &NiagaraSpriteRendererProperties]
  renderers [
   {rendererClass=&NiagaraLightRendererProperties, rendererIndex=0}
   {rendererClass=&NiagaraSpriteRendererProperties, rendererIndex=1}
  ]
 }
 MeshRibbon002 {
  emitterSpawnScript [/SetVariables(Emitter.MeshRibbonData=DataInterface() {propertyValues {
      bExposeCurve=true
      AlphaCurve {keys []}
      BlueCurve {keys [<at 0 1>,<at 1 1>]}
      GreenCurve {keys [<at 0 -0.5 RCIM_Cubic>, {arriveTangent=0.0331, interpMode=RCIM_Cubic, leaveTangent=0.0331, tangentMode=RCTM_User, time=1, value=-1}]}
      RedCurve {keys [<at 0 1 RCIM_Cubic>,<at 1 2 RCIM_Cubic>]}
     }}) {isSetParameters=true, module=SetVariables_#4}]
  emitterUpdateScript [
   &EmitterState()
   &SpawnBurst_Instantaneous(Spawn Count=12, Use Spawn Probability=false, UseLoopCountLimit=false) {enabled=false}
   &SpawnRate(
    SpawnRate=&SelectFloatFromArray(Direct Array Index=$System.Phase, Float Selection Array=DataInterface() {propertyValues {FloatData [0,0,0,0,0,0,2]}})
    Use Spawn Probability=false
   )
  ]
  particleSpawnScript [
   &InitializeParticle(
    Color=<rgba 1 0.3581 0 0.5>
    Lifetime=&RandomRangeFloat(Minimum=5.5268, Maximum=8.8377)
    Mesh Scale=&Multiply_VectorByFloat(Vector=<1 -1 1>, Float=&RandomRangeFloat(Minimum=12, Maximum=15))
    Mesh Scale Mode=::NewEnumerator4
    Position Offset=<0 -300 0>
    Position Offset Coordinate Space=::World
    UsePositionOffset=true
    Write Material Random=true
   )
   &InitialMeshOrientation(
    Mesh Orientation Mode=::NewEnumerator2
    Orientation Vector=<0 0 -1>
    Rotation=&Add_Vector(A=<0 0 0.15>, B=&RandomRangeVector(Minimum=<0 0 0>, Maximum=<0 0 0.6>))
    Use Rotation Vector=true
   )
   &InitialMeshOrientation(Mesh Orientation Mode=::NewEnumerator2, Orientation Vector=$Particles.ShapeLocation.ShapeNormal, Use Rotation Vector=false) {enabled=false, module=InitialMeshOrientation001}
  ]
  particleUpdateScript [
   &ParticleState()
   &ScaleMeshSize(Scale Factor=&VectorFromFloat(Value=&FloatFromCurve(FloatCurve=DataInterface() {propertyValues {Curve {keys [{arriveTangent=0.8366, interpMode=RCIM_Cubic, leaveTangent=0.8366, tangentMode=RCTM_User, time=0, value=0.5}, <at 1 0.75 RCIM_Cubic>]}}})))
   &DynamicMaterialParameters(
    Index 0 Param 1=&FloatFromCurve(FloatCurve=DataInterface() {propertyValues {Curve {keys [{tangentMode=RCTM_User, time=0, value=-1}, {tangentMode=RCTM_User, time=1, value=0}]}}})
    Index 0 Param 2=&FloatFromCurve(FloatCurve=DataInterface() {propertyValues {Curve {keys [{tangentMode=RCTM_User, time=0, value=0}, {tangentMode=RCTM_User, time=1, value=1}]}}})
    Index 0 Param 3=&RandomRangeFloat(Minimum=2.5103, Maximum=7.4856)
    Index 0 Param 4=&Add_Float(A=&RandomRangeFloat(Minimum=25.1137, Maximum=44.0696), B=0)
    Module.Index0Param0Write=true
    Module.Index0Param1Write=true
    Module.Index0Param2Write=true
    Module.Index0Param3Write=true
   )
   &ScaleColor(
    Linear Color Curve=DataInterface() {propertyValues {
      AlphaCurve {keys [<at 0 0 RCIM_Cubic>,<at 0.6722 5 RCIM_Cubic>,<at 1 0 RCIM_Cubic>]}
      BlueCurve {keys [<at 0 0.9922 RCIM_Cubic>, {arriveTangent=0.0078, interpMode=RCIM_Cubic, leaveTangent=0.0078, time=0.6722, value=5}, <at 1 1 RCIM_Cubic>]}
      GreenCurve {keys [<at 0 0.9941 RCIM_Cubic>, {arriveTangent=0.0059, interpMode=RCIM_Cubic, leaveTangent=0.0059, time=0.6722, value=5}, <at 1 1 RCIM_Cubic>]}
      RedCurve {keys [<at 0 1 RCIM_Cubic>,<at 0.6722 5 RCIM_Cubic>,<at 1 1 RCIM_Cubic>]}
     }}
    Scale Mode=::NewEnumerator2
   )
   &CameraOffset(Camera Offset Amount=50)
  ]
  rendererClasses [&NiagaraMeshRendererProperties]
  renderers [{rendererClass=&NiagaraMeshRendererProperties, rendererIndex=0}]
 }
 MeshRibbon003 {
  emitterSpawnScript [/SetVariables(Emitter.MeshRibbonData=DataInterface() {propertyValues {
      bExposeCurve=true
      AlphaCurve {keys []}
      BlueCurve {keys [<at 0 1>,<at 1 1>]}
      GreenCurve {keys [<at 0 -0.5 RCIM_Cubic>, {arriveTangent=0.0331, interpMode=RCIM_Cubic, leaveTangent=0.0331, tangentMode=RCTM_User, time=1, value=-1}]}
      RedCurve {keys [<at 0 1 RCIM_Cubic>,<at 1 2 RCIM_Cubic>]}
     }}) {isSetParameters=true, module=SetVariables_#4}]
  emitterUpdateScript [
   &EmitterState()
   &SpawnBurst_Instantaneous(Spawn Count=12, Use Spawn Probability=false, UseLoopCountLimit=false) {enabled=false}
   &SpawnRate(
    SpawnRate=&SelectFloatFromArray(Direct Array Index=$System.Phase, Float Selection Array=DataInterface() {propertyValues {FloatData [0,0,0,0,0,0,2]}})
    Use Spawn Probability=false
   )
  ]
  particleSpawnScript [
   &InitializeParticle(
    Color=<rgba 1 0.3581 0 0.5>
    Lifetime=&RandomRangeFloat(Minimum=5.5268, Maximum=8.8377)
    Mesh Scale=&Multiply_VectorByFloat(Vector=<1 -1 1>, Float=&RandomRangeFloat(Minimum=7, Maximum=12))
    Mesh Scale Mode=::NewEnumerator4
    Position Offset=<0 -200 0>
    Position Offset Coordinate Space=::World
    UsePositionOffset=true
    Write Material Random=true
   )
   &InitialMeshOrientation(
    Mesh Orientation Mode=::NewEnumerator2
    Orientation Vector=<0 0 -1>
    Rotation=&Add_Vector(A=<0 0 0.15>, B=&RandomRangeVector(Minimum=<0 0 0>, Maximum=<0 0 0.6>))
    Use Rotation Vector=true
   )
   &InitialMeshOrientation(Mesh Orientation Mode=::NewEnumerator2, Orientation Vector=$Particles.ShapeLocation.ShapeNormal, Use Rotation Vector=false) {enabled=false, module=InitialMeshOrientation001}
  ]
  particleUpdateScript [
   &ParticleState()
   &ScaleMeshSize(Scale Factor=&VectorFromFloat(Value=&FloatFromCurve(FloatCurve=DataInterface() {propertyValues {Curve {keys [{arriveTangent=0.8366, interpMode=RCIM_Cubic, leaveTangent=0.8366, tangentMode=RCTM_User, time=0, value=0.5}, <at 1 0.75 RCIM_Cubic>]}}})))
   &DynamicMaterialParameters(
    Index 0 Param 1=&FloatFromCurve(FloatCurve=DataInterface() {propertyValues {Curve {keys [{tangentMode=RCTM_User, time=0, value=-1}, {tangentMode=RCTM_User, time=1, value=0}]}}})
    Index 0 Param 2=&FloatFromCurve(FloatCurve=DataInterface() {propertyValues {Curve {keys [{tangentMode=RCTM_User, time=0, value=0}, {tangentMode=RCTM_User, time=1, value=1}]}}})
    Index 0 Param 3=&RandomRangeFloat(Minimum=2.5103, Maximum=7.4856)
    Index 0 Param 4=&Add_Float(A=&RandomRangeFloat(Minimum=25.1137, Maximum=44.0696), B=0)
    Module.Index0Param0Write=true
    Module.Index0Param1Write=true
    Module.Index0Param2Write=true
    Module.Index0Param3Write=true
   )
   &ScaleColor(
    Linear Color Curve=DataInterface() {propertyValues {
      AlphaCurve {keys [<at 0 0 RCIM_Cubic>,<at 0.6722 5 RCIM_Cubic>,<at 1 0 RCIM_Cubic>]}
      BlueCurve {keys [<at 0 0.9922 RCIM_Cubic>, {arriveTangent=0.0078, interpMode=RCIM_Cubic, leaveTangent=0.0078, time=0.6722, value=5}, <at 1 1 RCIM_Cubic>]}
      GreenCurve {keys [<at 0 0.9941 RCIM_Cubic>, {arriveTangent=0.0059, interpMode=RCIM_Cubic, leaveTangent=0.0059, time=0.6722, value=5}, <at 1 1 RCIM_Cubic>]}
      RedCurve {keys [<at 0 1 RCIM_Cubic>,<at 0.6722 5 RCIM_Cubic>,<at 1 1 RCIM_Cubic>]}
     }}
    Scale Mode=::NewEnumerator2
   )
   &CameraOffset(Camera Offset Amount=50)
  ]
  rendererClasses [&NiagaraMeshRendererProperties]
  renderers [{rendererClass=&NiagaraMeshRendererProperties, rendererIndex=0}]
 }
 Minimal {
  emitterUpdateScript [
   &EmitterState()
   &SpawnBurst_Instantaneous(Use Spawn Probability=false, UseLoopCountLimit=false)
   &SpawnBurst_Instantaneous(Spawn Time=0.8833, Use Spawn Probability=false, UseLoopCountLimit=false) {module=SpawnBurst_Instantaneous001}
  ]
  particleSpawnScript [&InitializeParticle(
    Color=&MakeCustomLinearColorFromBool(
     Bool=&SetBoolByIntComparison(A=$System.Phase, B=6, Comparison Type=::NewEnumerator4)
     True Color=<rgba 0.6406 0.3119 0.0606 1>
     False Color=<rgba 0.6406 0.0606 0.2832 1>
    )
    Lifetime=5
    Mesh Scale=<25 25 25>
    Mesh Scale Mode=::NewEnumerator4
    UsePositionOffset=true
    Write Material Random=true
   )]
  particleUpdateScript [
   &ParticleState()
   &ScaleMeshSize()
   &ScaleColor(
    Scale Alpha=&SelectFloatFromArray(Direct Array Index=$System.Phase, Float Selection Array=DataInterface() {propertyValues {FloatData [0,0.25,0.5,0.75,1,0,1]}})
    ScaleA=true
    ScaleRGB=true
   )
  ]
  rendererClasses [&NiagaraMeshRendererProperties]
  renderers [{rendererClass=&NiagaraMeshRendererProperties, rendererIndex=0}]
 }
 SparksLayer1 {
  simTarget=GPUComputeSim
  emitterUpdateScript [
   &EmitterState(Life Cycle Mode=::NewEnumerator1, Loop Duration=1, UseLoopDelay=false)
   &SpawnRate(
    SpawnRate=&SelectFloatFromArray(Direct Array Index=$System.Phase, Float Selection Array=DataInterface() {propertyValues {FloatData [0,100,500,1000,2000,0,0]}})
    Use Spawn Probability=false
   )
  ]
  particleSpawnScript [
   &InitializeParticle(
    Lifetime=&FloatFromCurve(
     FloatCurve=DataInterface() {propertyValues {Curve {keys [
         {interpMode=RCIM_Cubic, tangentMode=RCTM_User, time=0, value=3}
         {interpMode=RCIM_Cubic, tangentMode=RCTM_User, time=0.5, value=7}
         {tangentMode=RCTM_User, time=1, value=3}
        ]}}}
     CurveIndex=&Frac_Float(Float=$Engine.Time)
    )
    Mass Mode=::NewEnumerator1
    Sprite Size=&Multiply_Vector2(
     A=&Vector2DFromFloat(Value=&UniformRangedFloat(Minimum=0.4, Maximum=0.6))
     B=<2 6>
    )
    Sprite Size Mode=::NewEnumerator4
    UsePositionOffset=false
    Write Material Random=true
   ) {module=InitializeParticle001}
   &CylinderLocation(
    Cylinder Height=1
    Cylinder Radius=$System.SparkRadius
    Hemisphere X=true
    Hemisphere Y=true
    MaskSpawnGroup=false
    Module.Enable Lathe Profile=false
    Module.Override Randomness=false
    Module.Surface Only=true
    Orientation Axis:ENiagaraOrientationAxis=::XAxis
    Use Endcaps In Surface Only Mode=false
   )
   &self:InitialPosCHange(Angle:NiagaraFloat=&Multiply_Float(A=180, B=$Engine.Time)) {module=InitialPosCHange}
   /SetVariables(Transient.Position2:NiagaraPosition=$Particles.Position) {isSetParameters=true, module=SetVariables_#5}
   &CylinderLocation(
    Cylinder Height=1
    Cylinder Origin:Vector3f=&SimulationPosition()
    Cylinder Radius=$System.SparkRadius
    Lathe Profile=DataInterface() {propertyValues {Curve {keys [<at 0 0.451 RCIM_Cubic>, {arriveTangent=0.049, interpMode=RCIM_Cubic, leaveTangent=0.049, time=0.5, value=1}, <at 1 0.5 RCIM_Cubic>]}}}
    MaskSpawnGroup=false
    Module.Enable Lathe Profile=false
    Module.Override Randomness=false
    Module.Surface Only=true
    Orientation Axis:ENiagaraOrientationAxis=::XAxis
    Use Endcaps In Surface Only Mode=false
   ) {module=CylinderLocation001}
   &self:ChoosePos(Lerp=0.3) {module=ChoosePos}
   /Niagara/Modules/Update/Forces/CurlNoiseForce(Noise Frequency=0.005, Noise Quality / Cost=::NewEnumerator3, Noise Strength=50, Pan Noise=true, Pan Noise Field=<1 0 0>) {module=CurlNoiseForce001}
   &PointForce(Force Strength=-250, Module.Use Falloff=false, Module.UseLinearForce=false) {module=PointForce001}
   &ApplyInitialForces()
   /SetVariables(Particles.Position=&Add_Vector(A=Linked() {linkedVariable {name=Particles.Position, type=NiagaraPosition}}, B=$Transient.PhysicsForce)) {isSetParameters=true, module=SetVariables_#6}
  ]
  particleUpdateScript [
   &ParticleState()
   &CurlNoiseForce(Noise Frequency=0.0015, Noise Quality / Cost=::NewEnumerator3, Noise Strength=7.5, Pan Noise=true, Pan Noise Field=<0.5 0 0>)
   &self:ResampleFroce(Intensity=3) {module=ResampleFroce}
   &VortexForce(Influence Falloff Exponent=1.5, Influence Falloff Radius=1500, Use Influence Falloff=true, Vortex Axis=<1 0 0>, Vortex Force Amount=800)
   &PointForce(Force Falloff Distance=1000, Force Strength=400, Module.Use Falloff=true, Module.UseLinearForce=false)
   &Drag(
    Drag=&FloatFromCurve(
     FloatCurve=DataInterface() {propertyValues {Curve {keys [
         <at -0.0101 2.7154>
         {arriveTangent=-0.3026, interpMode=RCIM_Cubic, leaveTangent=-0.3026, tangentMode=RCTM_User, time=0.0352, value=1.1374}
         {arriveTangent=1.8448, interpMode=RCIM_Cubic, leaveTangent=1.8448, time=0.5183, value=1.5087}
         {arriveTangent=3.0746, interpMode=RCIM_Cubic, leaveTangent=3.0746, tangentMode=RCTM_User, time=0.8151, value=2.5761}
        ]}}}
     CurveIndex=&Divide_Float(A=$Particles.Age, B=6.5)
    )
    Use Linear Drag=true
    Use Rotational Drag=false
   )
   &SolveForcesAndVelocity(Clamp Velocity=false, Limit Acceleration=false)
   &ScaleColor(
    Scale Alpha=&FloatFromCurve(FloatCurve=DataInterface() {propertyValues {Curve {keys [<at 0 0>,<at 0.085 1>,<at 0.85 1>,<at 1 0>]}}})
    Scale RGB=&Lerp_Vector(
     A=&MakeVectorFromLinearColorRGB(Color=&ColorFromCurve(ColorCurve=DataInterface() {propertyValues {
         AlphaCurve {keys [<at 0 1 RCIM_Cubic>,<at 1 0 RCIM_Cubic>]}
         BlueCurve {keys [<at 0.089 2.7858 RCIM_Cubic>,<at 0.244 0.2048>,<at 0.629 0>]}
         GreenCurve {keys [<at 0.089 0.6772 RCIM_Cubic>,<at 0.244 0.0263>,<at 0.629 0>]}
         RedCurve {keys [<at 0.089 10 RCIM_Cubic>,<at 0.244 1>,<at 0.629 0>]}
        }}))
     B=&MakeVectorFromLinearColorRGB(Color=&ColorFromCurve(ColorCurve=DataInterface() {propertyValues {
         AlphaCurve {keys [<at 0 1 RCIM_Cubic>,<at 1 0 RCIM_Cubic>]}
         BlueCurve {keys [<at 0.089 1.1819 RCIM_Cubic>,<at 0.244 0>,<at 0.629 0>]}
         GreenCurve {keys [<at 0.089 5.5535 RCIM_Cubic>,<at 0.244 0.4598>,<at 0.629 0>]}
         RedCurve {keys [<at 0.089 10 RCIM_Cubic>,<at 0.244 1>,<at 0.629 0>]}
        }}))
     Alpha=&MakeCustomFloatFromBool(Bool=&SetBoolByIntComparison(A=$System.Phase, B=6, Comparison Type=::NewEnumerator4), True Float=1, False Float=0)
    )
    ScaleA=true
    ScaleRGB=true
   ) {module=ScaleColor001}
   &ScaleSpriteSizeBySpeed(Max Scale Factor=<6 10>, Min Scale Factor=<6 3>, Velocity Threshold=300)
   &self:MatchScale() {module=MatchScale}
   &CameraOffset(Camera Offset Amount=&FloatFromCurve(
     FloatCurve=DataInterface() {propertyValues {Curve {keys [{tangentMode=RCTM_User, time=0, value=0}, {tangentMode=RCTM_User, time=1, value=2}]}}}
     Scale Curve=25
    ))
  ]
  rendererClasses [&NiagaraSpriteRendererProperties]
  renderers [{rendererClass=&NiagaraSpriteRendererProperties}]
 }
 SparksLayer_Gold {
  simTarget=GPUComputeSim
  emitterUpdateScript [
   &EmitterState(Life Cycle Mode=::NewEnumerator1, Loop Duration=1, Recalculate Duration Each Loop=true, UseLoopDelay=false)
   &SpawnRate(
    SpawnRate=&Multiply_Float(
     A=&Waveform("Waveform [1]"=::NewEnumerator12, "[1] Frequency"=2, "[1] Amplitude Min/Max"=<-0.8 1>)
     B=&SelectFloatFromArray(Direct Array Index=$System.Phase, Float Selection Array=DataInterface() {propertyValues {FloatData [0,0,0,0,0,0,1000]}})
    )
    Use Spawn Probability=false
   )
  ]
  particleSpawnScript [
   &InitializeParticle(
    Lifetime=&FloatFromCurve(
     FloatCurve=DataInterface() {propertyValues {Curve {keys [
         {interpMode=RCIM_Cubic, tangentMode=RCTM_User, time=0, value=1}
         {interpMode=RCIM_Cubic, tangentMode=RCTM_User, time=0.5, value=2}
         {tangentMode=RCTM_User, time=1, value=1}
        ]}}}
     CurveIndex=&Frac_Float(Float=$Engine.Time)
    )
    Mass Mode=::NewEnumerator1
    Sprite Size=&Multiply_Vector2(
     A=&Vector2DFromFloat(Value=&UniformRangedFloat(Minimum=0.4, Maximum=0.6))
     B=<2 6>
    )
    Sprite Size Mode=::NewEnumerator4
    UsePositionOffset=true
    Write Material Random=true
   ) {module=InitializeParticle001}
   &ShapeLocation(
    Cylinder Distribution=::NewEnumerator2
    Cylinder Height=0
    Cylinder Radius=&RandomRangeFloat(Minimum=600, Maximum=1000)
    MaskSpawnGroup=false
    Rotation Mode=::NewEnumerator1
    Shape Primitive=::NewEnumerator1
    U Position=&Multiply_Float(A=$Emitter.Age, B=0.5)
    UseOwnerScale=true
    V Position=0
   )
   /SetVariables(Transient.Position2:NiagaraPosition=$Particles.Position) {isSetParameters=true, module=SetVariables_#5}
   &self:ChoosePos(Lerp=0.3) {module=ChoosePos}
   &CurlNoiseForce(Noise Frequency=0.005, Noise Quality / Cost=::NewEnumerator3, Noise Strength=50, Pan Noise=true, Pan Noise Field=<1 0 0>) {module=CurlNoiseForce001}
   &PointForce(Force Strength=250, Module.Use Falloff=false, Module.UseLinearForce=false) {module=PointForce001}
   &ApplyInitialForces()
   /SetVariables(Particles.Position=&Add_Vector(A=Linked() {linkedVariable {name=Particles.Position, type=NiagaraPosition}}, B=$Transient.PhysicsForce)) {isSetParameters=true, module=SetVariables_#6}
   &AddVelocity(Rotation Coordinate Space=::World, Velocity=&RandomRangeVector(Minimum=<0 0 0>, Maximum=<0 -1000 1>))
  ]
  particleUpdateScript [
   &ParticleState()
   &CurlNoiseForce(Noise Frequency=0.0015, Noise Quality / Cost=::NewEnumerator3, Noise Strength=7.5, Pan Noise=true, Pan Noise Field=<0.5 0 0>)
   &VortexForce(
    Influence Falloff Exponent=1.5
    Influence Falloff Radius=1500
    Invert Influence Falloff=false
    Use Influence Falloff=true
    Vortex Axis=<-1 0 0>
    Vortex Force Amount=-1000
   )
   &Drag(
    Drag=&FloatFromCurve(
     FloatCurve=DataInterface() {propertyValues {Curve {keys [
         <at -0.0101 2.7154>
         {arriveTangent=-0.3026, interpMode=RCIM_Cubic, leaveTangent=-0.3026, tangentMode=RCTM_User, time=0.0352, value=1.1374}
         {arriveTangent=1.8448, interpMode=RCIM_Cubic, leaveTangent=1.8448, time=0.5183, value=1.5087}
         {arriveTangent=3.0746, interpMode=RCIM_Cubic, leaveTangent=3.0746, tangentMode=RCTM_User, time=0.8151, value=2.5761}
        ]}}}
     CurveIndex=&Divide_Float(A=$Particles.Age, B=6.5)
    )
    Use Linear Drag=true
    Use Rotational Drag=false
   )
   &SolveForcesAndVelocity(Clamp Velocity=false, Limit Acceleration=false)
   &ScaleColor(
    Scale Alpha=&FloatFromCurve(FloatCurve=DataInterface() {propertyValues {Curve {keys [<at 0 0>,<at 0.085 1>,<at 0.85 1>,<at 1 0>]}}})
    Scale RGB=&Lerp_Vector(
     A=&MakeVectorFromLinearColorRGB(Color=&ColorFromCurve(ColorCurve=DataInterface() {propertyValues {
         AlphaCurve {keys [<at 0 1 RCIM_Cubic>,<at 1 0 RCIM_Cubic>]}
         BlueCurve {keys [<at 0.089 2.7858 RCIM_Cubic>,<at 0.244 0.2048>,<at 0.629 0>]}
         GreenCurve {keys [<at 0.089 0.6772 RCIM_Cubic>,<at 0.244 0.0263>,<at 0.629 0>]}
         RedCurve {keys [<at 0.089 10 RCIM_Cubic>,<at 0.244 1>,<at 0.629 0>]}
        }}))
     B=&MakeVectorFromLinearColorRGB(Color=&ColorFromCurve(ColorCurve=DataInterface() {propertyValues {
         AlphaCurve {keys [<at 0 1 RCIM_Cubic>,<at 1 0 RCIM_Cubic>]}
         BlueCurve {keys [<at 0.089 1.1819 RCIM_Cubic>,<at 0.244 0>,<at 0.629 0>]}
         GreenCurve {keys [<at 0.089 5.5535 RCIM_Cubic>,<at 0.244 0.4598>,<at 0.629 0>]}
         RedCurve {keys [<at 0.089 10 RCIM_Cubic>,<at 0.244 1>,<at 0.629 0>]}
        }}))
     Alpha=&MakeCustomFloatFromBool(Bool=&SetBoolByIntComparison(A=$System.Phase, B=6, Comparison Type=::NewEnumerator4), True Float=1, False Float=0)
    )
    ScaleA=true
    ScaleRGB=true
   ) {module=ScaleColor001}
   &ScaleSpriteSizeBySpeed(Max Scale Factor=<6 10>, Min Scale Factor=<6 3>, Velocity Threshold=300)
   &self:MatchScale() {module=MatchScale}
   &CameraOffset(Camera Offset Amount=&FloatFromCurve(
     FloatCurve=DataInterface() {propertyValues {Curve {keys [{tangentMode=RCTM_User, time=0, value=0}, {tangentMode=RCTM_User, time=1, value=2}]}}}
     Scale Curve=25
    ))
  ]
  rendererClasses [&NiagaraSpriteRendererProperties]
  renderers [{rendererClass=&NiagaraSpriteRendererProperties}]
 }
 SparksLayer_Gold_close {
  simTarget=GPUComputeSim
  emitterUpdateScript [
   &EmitterState(Life Cycle Mode=::NewEnumerator1, Loop Duration=1, UseLoopDelay=false)
   &SpawnRate(
    SpawnRate=&Multiply_Float(
     A=&Waveform("Waveform [1]"=::NewEnumerator12, "[1] Frequency"=2, "[1] Amplitude Min/Max"=<-0.5 1>)
     B=&SelectFloatFromArray(Direct Array Index=$System.Phase, Float Selection Array=DataInterface() {propertyValues {FloatData [0,0,0,0,0,0,1000]}})
    )
    Use Spawn Probability=false
   )
  ]
  particleSpawnScript [
   &InitializeParticle(
    Lifetime=&FloatFromCurve(
     FloatCurve=DataInterface() {propertyValues {Curve {keys [
         {interpMode=RCIM_Cubic, tangentMode=RCTM_User, time=0, value=3}
         {interpMode=RCIM_Cubic, tangentMode=RCTM_User, time=0.5, value=7}
         {tangentMode=RCTM_User, time=1, value=3}
        ]}}}
     CurveIndex=&Frac_Float(Float=$Engine.Time)
    )
    Mass Mode=::NewEnumerator1
    Position Offset=<0 0 100>
    Position Offset Coordinate Space=::World
    Sprite Size=&Multiply_Vector2(
     A=&Vector2DFromFloat(Value=&UniformRangedFloat(Minimum=0.4, Maximum=0.6))
     B=<2 6>
    )
    Sprite Size Mode=::NewEnumerator4
    UsePositionOffset=true
    Write Material Random=true
   ) {module=InitializeParticle001}
   &ShapeLocation(
    Cylinder Distribution=::NewEnumerator2
    Cylinder Height=0
    Cylinder Radius=&RandomRangeFloat(Minimum=600, Maximum=1000)
    MaskSpawnGroup=false
    Rotation Mode=::NewEnumerator1
    Shape Primitive=::NewEnumerator1
    U Position=&Multiply_Float(A=$Emitter.Age, B=0.25)
    UseOwnerScale=true
    V Position=0
   )
   /SetVariables(Transient.Position2:NiagaraPosition=$Particles.Position) {isSetParameters=true, module=SetVariables_#5}
   &CylinderLocation(
    Cylinder Height=1
    Cylinder Origin:Vector3f=&SimulationPosition()
    Cylinder Radius=&RandomRangeFloat(Minimum=500, Maximum=1000)
    Lathe Profile=DataInterface() {propertyValues {Curve {keys [<at 0 0.451 RCIM_Cubic>, {arriveTangent=0.049, interpMode=RCIM_Cubic, leaveTangent=0.049, time=0.5, value=1}, <at 1 0.5 RCIM_Cubic>]}}}
    MaskSpawnGroup=false
    Module.Enable Lathe Profile=false
    Module.Override Randomness=false
    Module.Surface Only=true
    Orientation Axis:ENiagaraOrientationAxis=::XAxis
    Use Endcaps In Surface Only Mode=false
   ) {module=CylinderLocation001}
   &self:ChoosePos(Lerp=0.3) {module=ChoosePos}
   &CurlNoiseForce(Noise Frequency=0.005, Noise Quality / Cost=::NewEnumerator3, Noise Strength=50, Pan Noise=true, Pan Noise Field=<1 0 0>) {module=CurlNoiseForce001}
   &PointForce(Force Strength=250, Module.Use Falloff=false, Module.UseLinearForce=false) {module=PointForce001}
   &ApplyInitialForces()
   /SetVariables(Particles.Position=&Add_Vector(A=Linked() {linkedVariable {name=Particles.Position, type=NiagaraPosition}}, B=$Transient.PhysicsForce)) {isSetParameters=true, module=SetVariables_#6}
  ]
  particleUpdateScript [
   &ParticleState()
   &CurlNoiseForce(Noise Frequency=0.0015, Noise Quality / Cost=::NewEnumerator3, Noise Strength=7.5, Pan Noise=true, Pan Noise Field=<0.5 0 0>)
   &self:ResampleFroce(Intensity=3) {module=ResampleFroce}
   &VortexForce(Influence Falloff Exponent=1.5, Influence Falloff Radius=1500, Use Influence Falloff=true, Vortex Axis=<1 0 0>, Vortex Force Amount=800)
   &PointForce(Force Falloff Distance=1000, Force Strength=400, Module.Use Falloff=true, Module.UseLinearForce=false)
   &Drag(
    Drag=&FloatFromCurve(
     FloatCurve=DataInterface() {propertyValues {Curve {keys [
         <at -0.0101 2.7154>
         {arriveTangent=-0.3026, interpMode=RCIM_Cubic, leaveTangent=-0.3026, tangentMode=RCTM_User, time=0.0352, value=1.1374}
         {arriveTangent=1.8448, interpMode=RCIM_Cubic, leaveTangent=1.8448, time=0.5183, value=1.5087}
         {arriveTangent=3.0746, interpMode=RCIM_Cubic, leaveTangent=3.0746, tangentMode=RCTM_User, time=0.8151, value=2.5761}
        ]}}}
     CurveIndex=&Divide_Float(A=$Particles.Age, B=6.5)
    )
    Use Linear Drag=true
    Use Rotational Drag=false
   )
   &SolveForcesAndVelocity(Clamp Velocity=false, Limit Acceleration=false)
   &ScaleColor(
    Scale Alpha=&FloatFromCurve(FloatCurve=DataInterface() {propertyValues {Curve {keys [<at 0 0>,<at 0.085 1>,<at 0.85 1>,<at 1 0>]}}})
    Scale RGB=&Lerp_Vector(
     A=&MakeVectorFromLinearColorRGB(Color=&ColorFromCurve(ColorCurve=DataInterface() {propertyValues {
         AlphaCurve {keys [<at 0 1 RCIM_Cubic>,<at 1 0 RCIM_Cubic>]}
         BlueCurve {keys [<at 0.089 2.7858 RCIM_Cubic>,<at 0.244 0.2048>,<at 0.629 0>]}
         GreenCurve {keys [<at 0.089 0.6772 RCIM_Cubic>,<at 0.244 0.0263>,<at 0.629 0>]}
         RedCurve {keys [<at 0.089 10 RCIM_Cubic>,<at 0.244 1>,<at 0.629 0>]}
        }}))
     B=&MakeVectorFromLinearColorRGB(Color=&ColorFromCurve(ColorCurve=DataInterface() {propertyValues {
         AlphaCurve {keys [<at 0 1 RCIM_Cubic>,<at 1 0 RCIM_Cubic>]}
         BlueCurve {keys [<at 0.089 1.1819 RCIM_Cubic>,<at 0.244 0>,<at 0.629 0>]}
         GreenCurve {keys [<at 0.089 5.5535 RCIM_Cubic>,<at 0.244 0.4598>,<at 0.629 0>]}
         RedCurve {keys [<at 0.089 10 RCIM_Cubic>,<at 0.244 1>,<at 0.629 0>]}
        }}))
     Alpha=&MakeCustomFloatFromBool(Bool=&SetBoolByIntComparison(A=$System.Phase, B=6, Comparison Type=::NewEnumerator4), True Float=1, False Float=0)
    )
    ScaleA=true
    ScaleRGB=true
   ) {module=ScaleColor001}
   &ScaleSpriteSizeBySpeed(Max Scale Factor=<6 10>, Min Scale Factor=<6 3>, Velocity Threshold=300)
   &self:MatchScale() {module=MatchScale}
   &CameraOffset(Camera Offset Amount=&FloatFromCurve(
     FloatCurve=DataInterface() {propertyValues {Curve {keys [{tangentMode=RCTM_User, time=0, value=0}, {tangentMode=RCTM_User, time=1, value=2}]}}}
     Scale Curve=25
    ))
  ]
  rendererClasses [&NiagaraSpriteRendererProperties]
  renderers [{rendererClass=&NiagaraSpriteRendererProperties}]
 }
 Spiral_Sparkles_Gold {
  simTarget=GPUComputeSim
  emitterUpdateScript [
   &EmitterState(Life Cycle Mode=::NewEnumerator1, Loop Behavior=::NewEnumerator1, Loop Duration Mode=::NewEnumerator1, UseLoopDelay=false)
   &SpawnRate(
    SpawnRate=&SelectFloatFromArray(Direct Array Index=$System.Phase, Float Selection Array=DataInterface() {propertyValues {FloatData [0,0,0,0,0,0,50]}})
    Use Spawn Probability=false
   )
  ]
  particleSpawnScript [
   &InitializeParticle(
    Color=&RandomRangeLinearColor(Minimum=<rgba 1 0.4962 0 1>, Maximum=<rgba 1 0.5969 0.2 1>)
    Lifetime=20
    Ribbon Width=200
    Ribbon Width Mode=::NewEnumerator1
    Sprite Size Mode=::NewEnumerator1
    Uniform Sprite Size Max=20
    UsePositionOffset=false
    Write Material Random=true
   ) {module=InitializeParticle001}
   &PartitionParticles(
    Particles Per Partition=100
    Partition Count Mode:ENiagara_InfiniteLoopDuration=::NewEnumerator1
    Partition Mode:ENiagara_PartitionMode=::NewEnumerator1
    Write Partition Index to Ribbon Link Order:ENiagara_AttributeSamplingApplyOutput=::NewEnumerator0
    Write Partition to Ribbon ID=::NewEnumerator0
   )
  ]
  particleUpdateScript [
   &ParticleState()
   &ShapeLocation(
    Handle Radius=447.528
    Large Radius=1137.7592
    MaskSpawnGroup=false
    Offset Mode=::NewEnumerator0
    Rotation Mode=::NewEnumerator2
    Shape Primitive=::NewEnumerator3
    Torus Distribution Mode:ENiagaraTorusDistributionMode=::NewEnumerator1
    U Position=&Add_Float(
     A=0.5
     B=&FloatFromCurve(FloatCurve=DataInterface() {propertyValues {Curve {keys [<at 0 0>,<at 1 1>]}}})
    )
    UseOwnerScale=true
    V Position=&Multiply_Float(A=$Particles.NormalizedAge, B=6)
    Yaw / Pitch / Roll=<0 0.25 0>
   )
   /SetVariables(Particles.Position=&Add_Vector(
     A=&RandomRangeVector(Minimum=<-50 -50 -50>, Maximum=<50 50 50>)
     B=Linked() {linkedVariable {name=Particles.Position, type=NiagaraPosition}}
    )) {isSetParameters=true}
   /SetVariables(Particles.RibbonFacing=$Particles.ShapeLocation.ShapeTangent) {isSetParameters=true, module=SetVariables_#7}
   &SolveForcesAndVelocity(Clamp Velocity=false, Limit Acceleration=false)
   &FN_RibbonMaterial(Color=<rgba 1 0.2481 0 1>, Color Pan U Speed=0.2, Micro Distortion Amount=1, Opacity Pan U Speed=0.1, Opacity Scale U=5)
   &ScaleRibbonWidth(Ribbon Width Scale=&FloatFromCurve(
     FloatCurve=DataInterface() {propertyValues {Curve {keys [
         {interpMode=RCIM_Cubic, tangentMode=RCTM_User, time=0, value=1}
         {arriveTangent=-2.1142, interpMode=RCIM_Cubic, leaveTangent=-2.1142, tangentMode=RCTM_User, time=0.1408, value=1.5}
         {arriveTangent=-0.0405, interpMode=RCIM_Cubic, leaveTangent=-0.0405, tangentMode=RCTM_User, time=1, value=1}
        ]}}}
     CurveIndex=&Frac_Float(Float=&Add_Float(
       A=&Multiply_Float(A=$Emitter.Age, B=0.5)
       B=$Particles.Initial.PartitionParticles.Normalized Index In Partition
      ))
     Scale Curve=0.5
    ))
   &ScaleColor(
    Scale RGB=&VectorFromFloat(Value=&FloatFromCurve(
      FloatCurve=DataInterface() {propertyValues {Curve {keys [
          {interpMode=RCIM_Cubic, tangentMode=RCTM_User, time=0, value=1}
          {arriveTangent=-2.1142, interpMode=RCIM_Cubic, leaveTangent=-2.1142, tangentMode=RCTM_User, time=0.1408, value=1.5}
          {arriveTangent=-0.0405, interpMode=RCIM_Cubic, leaveTangent=-0.0405, tangentMode=RCTM_User, time=1, value=1}
         ]}}}
      CurveIndex=&Frac_Float(Float=&Add_Float(A=&Multiply_Float(A=$Emitter.Age, B=0.5), B=$Particles.NormalizedAge))
     ))
    ScaleA=true
    ScaleRGB=true
   )
  ]
  rendererClasses [&NiagaraSpriteRendererProperties]
  renderers [{rendererClass=&NiagaraSpriteRendererProperties}]
 }
 SprialRibbon_Gold {
  emitterUpdateScript [
   &EmitterState(Life Cycle Mode=::NewEnumerator1, Loop Behavior=::NewEnumerator1, Loop Duration Mode=::NewEnumerator1, UseLoopDelay=false)
   &SpawnBurst_Instantaneous(
    Spawn Count=20
    Spawn Probability=&SelectFloatFromArray(Direct Array Index=$System.Phase, Float Selection Array=DataInterface() {propertyValues {FloatData [0,0,0,0,0,0,1]}})
    Use Spawn Probability=false
    UseLoopCountLimit=false
   )
  ]
  particleSpawnScript [
   &InitializeParticle(Color=<rgba 1 0.4962 0 1>, Ribbon Width=200, Ribbon Width Mode=::NewEnumerator1, UsePositionOffset=false, Write Material Random=true) {module=InitializeParticle001}
   &PartitionParticles(
    Particles Per Partition=20
    Partition Count Mode:ENiagara_InfiniteLoopDuration=::NewEnumerator1
    Partition Mode:ENiagara_PartitionMode=::NewEnumerator1
    Write Partition Index to Ribbon Link Order:ENiagara_AttributeSamplingApplyOutput=::NewEnumerator0
    Write Partition to Ribbon ID=::NewEnumerator0
   )
   &ShapeLocation(
    Handle Radius:NiagaraFloat=&FloatFromCurve(
     FloatCurve=DataInterface() {propertyValues {Curve {keys [<at 0 450>,<at 0.5 450>,<at 0.6 600>,<at 0.8 450>]}}}
     CurveIndex=$Particles.PartitionParticles.Normalized Index In Partition
    )
    Large Radius=&FloatFromCurve(
     FloatCurve=DataInterface() {propertyValues {Curve {keys [<at 0 1137>,<at 0.5 1137>,<at 0.6 1400>,<at 0.8 1137>]}}}
     CurveIndex=$Particles.PartitionParticles.Normalized Index In Partition
    )
    MaskSpawnGroup=false
    Offset Mode=::NewEnumerator0
    Rotation Mode=::NewEnumerator2
    Shape Primitive=::NewEnumerator3
    Torus Distribution Mode:ENiagaraTorusDistributionMode=::NewEnumerator1
    U Position=&Add_Float(
     A=0.5
     B=&FloatFromCurve(
      FloatCurve=DataInterface() {propertyValues {Curve {keys [<at 0 0.025>,<at 1 0.95>]}}}
      CurveIndex=$Particles.PartitionParticles.Normalized Index In Partition
     )
    )
    UseOwnerScale=true
    V Position=&Multiply_Float(A=$Particles.PartitionParticles.Normalized Index In Partition, B=6)
    Yaw / Pitch / Roll=<0 0.25 0>
   )
   /SetVariables(Particles.RibbonFacing=$Particles.ShapeLocation.ShapeTangent) {isSetParameters=true, module=SetVariables_#7}
  ]
  particleUpdateScript [
   &ParticleState(Kill Particles When Lifetime Has Elapsed=false)
   &SolveForcesAndVelocity(Clamp Velocity=false, Limit Acceleration=false)
   &FN_RibbonMaterial(Color=<rgba 1 0.5441 0 1>, Color Pan U Speed=0.2, Micro Distortion Amount=1, Opacity Pan U Speed=0.2, Opacity Scale U=5)
   &ScaleRibbonWidth(Ribbon Width Scale=&FloatFromCurve(
     FloatCurve=DataInterface() {propertyValues {Curve {keys [
         {interpMode=RCIM_Cubic, tangentMode=RCTM_User, time=0, value=1}
         {arriveTangent=-2.1142, interpMode=RCIM_Cubic, leaveTangent=-2.1142, tangentMode=RCTM_User, time=0.1408, value=1.5}
         {arriveTangent=-0.0405, interpMode=RCIM_Cubic, leaveTangent=-0.0405, tangentMode=RCTM_User, time=1, value=1}
        ]}}}
     CurveIndex=&Frac_Float(Float=&Add_Float(
       A=&Multiply_Float(A=$Emitter.Age, B=0.5)
       B=&OneMinusFloat(Float=$Particles.PartitionParticles.Normalized Index In Partition)
      ))
     Scale Curve=0.5
    ))
   &ScaleColor(
    Scale RGB=&VectorFromFloat(Value=&FloatFromCurve(
      FloatCurve=DataInterface() {propertyValues {Curve {keys [
          {interpMode=RCIM_Cubic, tangentMode=RCTM_User, time=0, value=1}
          {arriveTangent=-2.1142, interpMode=RCIM_Cubic, leaveTangent=-2.1142, tangentMode=RCTM_User, time=0.1408, value=1.5}
          {arriveTangent=-0.0405, interpMode=RCIM_Cubic, leaveTangent=-0.0405, tangentMode=RCTM_User, time=1, value=1}
         ]}}}
      CurveIndex=&Frac_Float(Float=&Add_Float(
        A=&Multiply_Float(A=$Emitter.Age, B=0.5)
        B=&OneMinusFloat(Float=$Particles.PartitionParticles.Normalized Index In Partition)
       ))
     ))
    ScaleA=true
    ScaleRGB=true
   )
   &ScaleRibbonWidth(Ribbon Width Scale=&SelectFloatFromArray(Direct Array Index=$System.Phase, Float Selection Array=DataInterface() {propertyValues {FloatData [0,0,0,0,0,0,1]}})) {module=ScaleRibbonWidth001}
  ]
  rendererClasses [&NiagaraRibbonRendererProperties]
  renderers [{rendererClass=&NiagaraRibbonRendererProperties}]
 }
 SubtleSmoke_R001 {
  simTarget=GPUComputeSim
  emitterUpdateScript [
   &EmitterState(
    Life Cycle Mode=::NewEnumerator1
    Loop Behavior=::NewEnumerator1
    Loop Delay=$System.RightHandDelay
    Loop Duration Mode=::NewEnumerator1
    UseLoopDelay=true
   )
   &SpawnRate(
    SpawnRate=&SelectFloatFromArray(Direct Array Index=$System.Phase, Float Selection Array=DataInterface() {propertyValues {FloatData [0,0,5,15,20]}})
    Use Spawn Probability=true
   )
  ]
  particleSpawnScript [
   &InitializeParticle(
    Lifetime Max=3
    Lifetime Min=2
    Lifetime Mode=::NewEnumerator1
    Mass Mode=::NewEnumerator3
    Sprite Rotation Mode=::NewEnumerator2
    Sprite Size Mode=::NewEnumerator3
    Uniform Sprite Size=&RandomRangeFloat(Minimum=200, Maximum=300)
    UsePositionOffset=false
    Write Material Random=true
   )
   /SetVariables(Particles.RandomFrame=&RandomRangeInt(Minimum=0, Maximum=63)) {isSetParameters=true}
   &CameraOffset()
   &ShapeLocation(
    Cylinder Radius=&Add_Float(
     A=&SelectFloatFromArray(Direct Array Index=$System.Phase, Float Selection Array=DataInterface() {propertyValues {FloatData [0,900,900,900,900]}})
     B=-250
    )
    MaskSpawnGroup=false
    Override Seed=false
    Rotation Angle=0.5
    Rotation Mode=::NewEnumerator1
    Shape Primitive=::NewEnumerator1
    Surface Only=true
    Use Endcaps In Surface Only Mode=false
    UseOwnerScale=true
   ) {module=ShapeLocation001}
   &AddVelocity(
    Module.Use Falloff Distance=false
    Origin Offset=<-500 0 0>
    Override Seed=false
    Velocity Mode=::NewEnumerator1
    Velocity Speed=&RandomRangeFloat(Minimum=100, Maximum=200)
   )
  ]
  particleUpdateScript [
   &ParticleState()
   &Drag(Use Linear Drag=true, Use Rotational Drag=true)
   &SolveForcesAndVelocity(Clamp Velocity=false, Limit Acceleration=false)
   &Color(Color=&LerpLinearColors(
     StartColor=&ColorFromCurve(ColorCurve=DataInterface() {propertyValues {
        AlphaCurve {keys [<at 0 0 RCIM_Cubic>,<at 0.2415 0.5>]}
        BlueCurve {keys [<at 0 2.1339 RCIM_Cubic>,<at 0.4698 0.099>]}
        GreenCurve {keys [<at 0 0.2954 RCIM_Cubic>,<at 0.4698 0.0632>]}
        RedCurve {keys [<at 0 10 RCIM_Cubic>,<at 0.4698 0.2526>]}
       }})
     EndColor=&ColorFromCurve(ColorCurve=DataInterface() {propertyValues {
        AlphaCurve {keys [<at 0 0 RCIM_Cubic>,<at 0.2415 0.5>]}
        BlueCurve {keys [<at 0 0.2954 RCIM_Cubic>,<at 0.4698 0.0632>]}
        GreenCurve {keys [<at 0 2.8833 RCIM_Cubic>,<at 0.4698 0.1421>]}
        RedCurve {keys [<at 0 10 RCIM_Cubic>,<at 0.4698 0.2526>]}
       }})
     LerpFactor=&MakeCustomFloatFromBool(Bool=&SetBoolByIntComparison(A=$System.Phase, B=6, Comparison Type=::NewEnumerator4), True Float=1, False Float=0)
    ))
   &ScaleColor(
    Color Value To Scale=$Particles.Color
    Scale Alpha=&FloatFromCurve(
     FloatCurve=DataInterface() {propertyValues {Curve {keys [
         {interpMode=RCIM_Cubic, tangentMode=RCTM_User, time=0, value=0}
         <at 0.8 1 RCIM_Cubic>
         {interpMode=RCIM_Cubic, tangentMode=RCTM_User, time=1, value=0}
        ]}}}
     Scale Curve=&FloatFromCurve(
      FloatCurve=DataInterface() {propertyValues {Curve {keys [{interpMode=RCIM_Cubic, tangentMode=RCTM_User, time=1, value=0}, {interpMode=RCIM_Cubic, tangentMode=RCTM_User, time=2, value=1}]}}}
      CurveIndex=$Emitter.Age
      Scale Curve=2
     )
    )
    ScaleA=true
    ScaleRGB=true
   )
   &DynamicMaterialParameters(
    Index 0 Param 1=&FloatFromCurve(
     FloatCurve=DataInterface() {propertyValues {Curve {keys [
         {arriveTangent=2.8718, interpMode=RCIM_Cubic, leaveTangent=2.8718, tangentMode=RCTM_User, time=0, value=0}
         {arriveTangent=-0.3942, interpMode=RCIM_Cubic, leaveTangent=-0.3942, tangentMode=RCTM_User, time=1, value=1}
        ]}}}
     Scale Curve=0.1
    )
    Index 0 Param 2=50
    Module.Index0Param0Write=true
    Module.Index0Param1Write=true
    Module.Index0Param2Write=true
    Module.Index0Param3Write=true
   )
   &FN_SubUVAnimation(End Frame=&Add_Integer(A=$Particles.RandomFrame, B=33), Start Frame=$Particles.RandomFrame)
   &ScaleColor(
    Color Value To Scale=$Particles.Color
    Scale Alpha=&SelectFloatFromArray(Direct Array Index=$System.Phase, Float Selection Array=DataInterface() {propertyValues {FloatData [0,0.2,0.2,0.5,1,0,1]}})
    ScaleA=true
    ScaleRGB=false
   ) {module=ScaleColor001}
  ]
  rendererClasses [&NiagaraSpriteRendererProperties]
  renderers [{rendererClass=&NiagaraSpriteRendererProperties}]
 }
}
summary {
 systemName=NS_GrillMoon_Portal
 emitters [
  {emitterName=SparksLayer1, simTarget=GPUComputeSim, rendererClasses [&NiagaraSpriteRendererProperties]}
  {emitterName=CenterPink, rendererClasses [&NiagaraSpriteRendererProperties]}
  {emitterName=Minimal, rendererClasses [&NiagaraMeshRendererProperties]}
  {emitterName=BG_Translucent, rendererClasses [&NiagaraSpriteRendererProperties, &NiagaraMeshRendererProperties]}
  {emitterName=BG_Opaque, rendererClasses [&NiagaraSpriteRendererProperties, &NiagaraMeshRendererProperties]}
  {emitterName=SubtleSmoke_R001, simTarget=GPUComputeSim, rendererClasses [&NiagaraSpriteRendererProperties]}
  {emitterName=SprialRibbon_Gold, rendererClasses [&NiagaraRibbonRendererProperties]}
  {emitterName=Spiral_Sparkles_Gold, simTarget=GPUComputeSim, rendererClasses [&NiagaraSpriteRendererProperties]}
  {emitterName=BG_Masked, rendererClasses [&NiagaraSpriteRendererProperties, &NiagaraMeshRendererProperties]}
  {emitterName=CenterGold, rendererClasses [&NiagaraSpriteRendererProperties]}
  {emitterName=SparksLayer_Gold, simTarget=GPUComputeSim, rendererClasses [&NiagaraSpriteRendererProperties]}
  {emitterName=SparksLayer_Gold_close, simTarget=GPUComputeSim, rendererClasses [&NiagaraSpriteRendererProperties]}
  {emitterName=LightAndGlow, rendererClasses [&NiagaraLightRendererProperties, &NiagaraSpriteRendererProperties]}
  {emitterName=MeshRibbon002, rendererClasses [&NiagaraMeshRendererProperties]}
  {emitterName=MeshRibbon003, rendererClasses [&NiagaraMeshRendererProperties]}
 ]
 userVariables [
  User.bIsFrontEnd {type=NiagaraBool}
  User.bIsFrontEndPreview {type=NiagaraBool}
  User.ChaosDestructionData {
   type=NiagaraDataInterfaceChaosDestruction
   description=""
   defaultValue {
    struct=NiagaraExt_VariableValue_DataInterface
    value {
     dataInterface=&self:NiagaraDataInterfaceChaosDestruction_0
     dataInterfaceClass=/Script/ChaosNiagara.NiagaraDataInterfaceChaosDestruction
    }
   }
  }
  User.Portal Looping Sound {
   type=NiagaraDataInterfaceAudioPlayer
   description=""
   defaultValue {
    struct=NiagaraExt_VariableValue_DataInterface
    value {
     dataInterface=&self:NiagaraDataInterfaceAudioPlayer_1
     dataInterfaceClass=&NiagaraDataInterfaceAudioPlayer
    }
   }
  }
  User.Skeletal Mesh {
   type=NiagaraDataInterfaceSkeletalMesh
   defaultValue {
    struct=NiagaraExt_VariableValue_DataInterface
    value {dataInterface=&self:NiagaraDataInterfaceSkeletalMesh_0}
   }
  }
  User.SkeletalMesh {
   type=NiagaraDataInterfaceSkeletalMesh
   defaultValue {
    struct=NiagaraExt_VariableValue_DataInterface
    value {dataInterface=&self:NiagaraDataInterfaceSkeletalMesh_1}
   }
  }
  User.WhipRope {type=NiagaraBool}
 ]
}
